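# UI material for a Crew Benefits tile with a two-color gradient background, a CMS-supplied image and rounded corners.
# The image pans, scales about its center and composites by alpha.
# A Hover Overlay parameter blends in a dodge-like brightening, and a sine-shaped gradient scaled by BottomFade drives an overlay fade.
# MF_UI_CoreUI_RoundedMask builds the shape mask and SDF from per-corner radii, and named reroutes link the stages.
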Material /Crew/UI/Materials/CrewBenefits/M_UI_CrewBenefitsTile
graph {
 properties {
  bAutomaticallySetUsageInEditor=false
  bUsedWithStaticMesh=false
  blendMode=BLEND_Translucent
  materialDomain=MD_UI
  usedShadingModels=Default Lit
 }
 nodes [
  n1 = NamedRerouteUsage()
  n2 = TextureCoordinate()
  n3 = ScalarParameter() {defaultValue=0.25, group=GradientOverlay, parameterName=GradOverlayStart}
  n4 = ScalarParameter() {defaultValue=0.9, group=GradientOverlay, parameterName=GradOverlayEnd}
  n5 = Constant() {r=0.75}
  n6 = ScalarParameter() {group=GradientOverlay, parameterName=GradOverlayOffset}
  n7 = Multiply(A=Subtract(A=MaterialFunctionCall(UVs=@n2, GradientStart=@n3, GradientEnd=@n4, RotationAngle=@n5, GradientOffset=@n6) {materialFunction=&MF_UI_LinearGradient}.Result) {constB=0.5}) {constB=0.5}
  n8 = Multiply(A=Add(A=Sine(None=@n7))) {constB=0.5}
  LinearInterpolate(
   A=@n1
   B=NamedRerouteUsage()
   Alpha=Multiply(A=Saturate(None=@n8), B=ScalarParameter() {defaultValue=0.8, group=Content Controller, parameterName=BottomFade})
  )
  NamedRerouteDeclaration(None=VectorParameter() {group=Color, parameterName=BackgroundColor1, defaultValue=<rgba 0.1779 0.4793 0.7682 1>}.RGB) {name=BackgroundColor1, nodeColor=<rgba 0 0 0 1>}
  NamedRerouteDeclaration(None=MaterialFunctionCall() {materialFunction=&GetUserInterfaceUV}.Normalized UV) {name=Normalized UV, nodeColor=<rgba 0 0 0 1>}
  n9 = ScalarParameter() {
   defaultValue=1
   desc="If enabled reads corner radius values in pixels.\r\nbool"
   group=Layout
   parameterName=UseValueInPixels
   sliderMax=1
   sortPriority=1
  }
  n10 = MaterialFunctionCall(
   Corner Radius=MaterialFunctionCall(
    R=ScalarParameter() {defaultValue=20, group=Layout, parameterName=TopLeftCornerRadius, sortPriority=4}
    G=ScalarParameter() {defaultValue=20, group=Layout, parameterName=BottomLeftCornerRadius, sortPriority=7}
    B=ScalarParameter() {defaultValue=20, group=Layout, parameterName=TopRightCornerRadius, sortPriority=5}
    A=ScalarParameter() {defaultValue=20, group=Layout, parameterName=BottomRightCornerRadius, sortPriority=6}
   ) {materialFunction=&AppendMany}.RGBA
   Radius in Pixels=@n9
  ) {materialFunction=&MF_UI_CoreUI_RoundedMask}
  NamedRerouteDeclaration(None=@n10.Shape Mask) {name=Shape Mask, nodeColor=<rgba 0 0 0 1>}
  n11 = NamedRerouteUsage()
  n12 = NamedRerouteUsage()
  n13 = Power(Base=@n12) {constExponent=0.2}
  n14 = Max(A=OneMinus(None=Multiply(A=@n13) {constB=0.2})) {constB=0}
  n15 = Add(A=Saturate(None=Divide(A=@n11, B=@n14)), B=Multiply(A=Power(Base=@n12) {constExponent=0.85}) {constB=0.3})
  NamedRerouteDeclaration(None=LinearInterpolate(A=@n11, B=Saturate(None=@n15), Alpha=ScalarParameter() {group=Hover, parameterName=Hover Overlay})) {name=Hover Overlay, nodeColor=<rgba 0 0 0 1>}
  NamedRerouteDeclaration(None=VectorParameter() {group=Color, parameterName=BackgroundColor2, defaultValue=<rgba 0.1022 0.2789 0.4072 1>}.RGB) {name=BackgroundColor2, nodeColor=<rgba 0 0 0 1>}
  NamedRerouteDeclaration(None=@n10.Outline SDF) {name=ShapeMask_SDF, nodeColor=<rgba 0 0 0 1>}
  NamedRerouteDeclaration(None=VectorParameter() {group=Color, parameterName=OverlayGradientColor, defaultValue=<rgba 0.0052 0.008 0.0144 1>}.RGB) {name=OverlayGradientColor, nodeColor=<rgba 0 0 0 1>}
  n16 = Constant() {r=0.75}
  NamedRerouteDeclaration(None=LinearInterpolate(
    A=NamedRerouteUsage()
    B=NamedRerouteUsage()
    Alpha=MaterialFunctionCall(RotationAngle=@n16) {materialFunction=&MF_UI_LinearGradient}.Result
   )) {name=BaseBackground, nodeColor=<rgba 0 0 0 1>}
  NamedRerouteDeclaration(None=@n1) {name=AllMerged, nodeColor=<rgba 0 0 0 1>}
  n17 = NamedRerouteUsage()
  n18 = AppendVector(A=ScalarParameter() {group=Texture, parameterName=TexturePanX}, B=ScalarParameter() {group=Texture, parameterName=TexturePanY})
  n19 = TextureSampleParameter2D(UVs=MaterialFunctionCall(UVs=Add(A=@n17, B=@n18), Texture Scale=ScalarParameter() {defaultValue=1, group=Texture, parameterName=TextureScale}) {materialFunction=&ScaleUVsByCenter}.UVs) {parameterName=CMSTexture, texture=/Crew/UI/Textures/T_UI_CrewBenefitsTile_Placeholder}
  NamedRerouteDeclaration(None=LinearInterpolate(A=NamedRerouteUsage(), B=@n19.RGB, Alpha=NamedRerouteDeclaration(None=@n19.A) {name=CMSTextureMask, nodeColor=<rgba 0 0 0 1>})) {name=Hover+Content, nodeColor=<rgba 0 0 0 1>}
  NamedRerouteUsage()
  NamedRerouteUsage()
 ]
}
