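# Translucent UI material for Crew benefits visuals, layering a gradient background, animated noise, a texture and a color overlay.
# A radial gradient with adjustable pan, rotation and scale mixes two colors, with time-seeded, capped random noise for overlay-blended grain.
# A pannable, scalable texture supplies color and alpha for compositing, and an inverted linear-gradient mask fades in the overlay color at adjustable strength.
# Named reroutes connect these stages into a chain of lerps.
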Material /Crew/UI/Materials/CrewBenefits/M_UI_CrewBenefits_Gradient
graph {
 properties {
  bAutomaticallySetUsageInEditor=false
  bUsedWithStaticMesh=false
  blendMode=BLEND_Translucent
  materialDomain=MD_UI
  usedShadingModels=Default Lit
 }
 nodes [
  n1 = NamedRerouteUsage()
  n2 = MaterialFunctionCall(Base=@n1, Blend=NamedRerouteUsage()) {materialFunction=&Blend_Overlay}
  n3 = ScalarParameter() {defaultValue=0.15, parameterName=NoiseAlpha}
  n4 = NamedRerouteDeclaration(None=OneMinus(None=NamedRerouteUsage())) {name=Overlay_Mask, nodeColor=<rgba 0 0 0 1>}
  LinearInterpolate(
   A=LinearInterpolate(A=LinearInterpolate(A=@n1, B=@n2.Result, Alpha=@n3) {constAlpha=0.3}, B=NamedRerouteUsage(), Alpha=NamedRerouteUsage())
   B=LinearInterpolate(B=NamedRerouteUsage(), Alpha=@n4)
   Alpha=Multiply(A=@n4, B=ScalarParameter() {defaultValue=1, group=Overlay, parameterName=Overlay_Strength}) {constB=0.6}
  )
  n5 = NamedRerouteUsage()
  n6 = ScalarParameter() {group=Gradient, parameterName=GradientStart}
  n7 = ScalarParameter() {defaultValue=1, group=Gradient, parameterName=GradientEnd}
  NamedRerouteDeclaration(None=LinearInterpolate(
    A=NamedRerouteUsage()
    B=NamedRerouteUsage()
    Alpha=MaterialFunctionCall(UV=@n5, GradientStart=@n6, GradientEnd=@n7) {materialFunction=&MF_UI_RadialGradient}.Smooth
   )) {name=BackgroundGradient, nodeColor=<rgba 0 0 0 1>}
  NamedRerouteDeclaration(None=MaterialFunctionCall!(Use Local UV=Constant() {r=1}, Texture Size, Scale Pivot Point) {materialFunction=/Game/Environments/FrontEnd/Blueprints/ItemPreview/VaultWorldTransitions/MaterialFunctions/MF_UI_VaultWorld_UV}.Scaled to Fit) {name=UV_ScaledToFit, nodeColor=<rgba 0 0 0 1>}
  NamedRerouteDeclaration(None=VectorParameter() {group=Color, parameterName=Color2, defaultValue=<rgba 0.1779 0.4793 0.7682 1>}.RGB) {name=BackgroundGradient_Color2, nodeColor=<rgba 0 0 0 1>}
  NamedRerouteDeclaration(None=VectorParameter() {group=Color, parameterName=Color4, defaultValue=<rgba 0.0052 0.008 0.0144 1>}.RGB) {name=Overlay_Color, nodeColor=<rgba 0 0 0 1>}
  n8 = Add(
   A=NamedRerouteUsage()
   B=AppendVector(A=ScalarParameter() {group=Gradient, parameterName=GradientXPan}, B=ScalarParameter() {group=Gradient, parameterName=GradientYPan})
  )
  n9 = ScalarParameter() {defaultValue=0.1, group=Gradient, parameterName=GradientRotation, sliderMax=0.5}
  n10 = AppendVector(
   A=ScalarParameter() {defaultValue=1.4, group=Gradient, parameterName=GradientXScale}
   B=ScalarParameter() {defaultValue=1, group=Gradient, parameterName=GradientYScale}
  )
  n11 = ScalarParameter() {defaultValue=2, group=Gradient, parameterName=GradientScale}
  NamedRerouteDeclaration(None=MaterialFunctionCall(
    UVs=MaterialFunctionCall(UVs=@n8, "Rotation Angle (0-1)"=@n9) {materialFunction=&CustomRotator}.Rotated Values
    Texture Scale=Multiply(A=@n10, B=@n11)
   ) {materialFunction=&ScaleUVsByCenter}.UVs) {name=UV_FocalPoint, nodeColor=<rgba 0 0 0 1>}
  n12 = Multiply(A=Time(), B=ScalarParameter() {defaultValue=0.5, parameterName=NoiseTime}) {constB=0.5}
  NamedRerouteDeclaration(None=Min(B=MaterialFunctionCall!(UV, Seed=@n12) {materialFunction=&MF_UI_Random}.Result) {constA=0.5}) {name=Noise, nodeColor=<rgba 0 0 0 1>}
  NamedRerouteDeclaration(None=MaterialFunctionCall(UVs=NamedRerouteUsage(), GradientStart=Constant(), GradientEnd=Constant() {r=0.5}, RotationAngle=Constant() {r=0.25}) {materialFunction=&MF_UI_LinearGradient}.Result) {name=Overlay, nodeColor=<rgba 0 0 0 1>}
  n13 = Add(
   A=NamedRerouteUsage()
   B=AppendVector(
    A=ScalarParameter() {defaultValue=-0.3, group=Texture, parameterName=Texture_XPan}
    B=ScalarParameter() {group=Texture, parameterName=Texture_YPan}
   )
  )
  n14 = ScalarParameter() {defaultValue=1, group=Texture, parameterName=Texture_Scale}
  n15 = TextureSampleParameter2D(UVs=Clamp(None=MaterialFunctionCall(UVs=@n13, Texture Scale=@n14) {materialFunction=&ScaleUVsByCenter}.UVs)) {group=Texture, parameterName=CMS_Texture, texture=/Game/UI/Foundation/Textures/Founders/T-VBucks-Details-5000}
  NamedRerouteDeclaration(None=@n15.A) {name=Texture_Alpha, nodeColor=<rgba 0 0 0 1>}
  NamedRerouteDeclaration(None=@n15.RGB) {name=Texture_Colors, nodeColor=<rgba 0 0 0 1>}
  NamedRerouteDeclaration(None=VectorParameter() {group=Color, parameterName=Color1, defaultValue=<rgba 0.0194 0.0452 0.0685 1>}.RGB) {name=BackgroundGradient_Color1, nodeColor=<rgba 0 0 0 1>}
 ]
}
