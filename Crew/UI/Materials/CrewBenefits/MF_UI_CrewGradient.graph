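# Crew benefits UI material function that composes a gradient background, CMS-supplied texture and color overlay into one output.
# A radial gradient with adjustable pan, rotation and scale blends two colors, with time-seeded random noise mixed in through Blend_Overlay.
# The texture is composited by alpha, using its dimensions and TargetAspectRatio to avoid squishing.
# An inverted linear-gradient mask applies the overlay color, scaled by Overlay_Strength.
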MaterialFunction /Crew/UI/Materials/CrewBenefits/MF_UI_CrewGradient
graph {nodes [
  Constant() {r=0.2}
  n1 = NamedRerouteUsage()
  n2 = LinearInterpolate(
   A=@n1
   B=MaterialFunctionCall(Base=@n1, Blend=NamedRerouteUsage()) {materialFunction=&Blend_Overlay}.Result
   Alpha=ScalarParameter() {defaultValue=0.15, parameterName=NoiseAlpha}
  ) {constAlpha=0.3}
  n3 = NamedRerouteUsage()
  n4 = NamedRerouteUsage()
  n5 = NamedRerouteUsage()
  n6 = NamedRerouteDeclaration(None=OneMinus(None=NamedRerouteUsage())) {name=Overlay_Mask, nodeColor=<rgba 0 0 0 1>}
  n7 = ScalarParameter() {defaultValue=1, group=Overlay, parameterName=Overlay_Strength}
  FunctionOutput(None=LinearInterpolate(A=LinearInterpolate(A=@n2, B=@n3, Alpha=@n4), B=LinearInterpolate(B=@n5, Alpha=@n6), Alpha=Multiply(A=@n6, B=@n7) {constB=0.6})) {outputName=Lerp}
  n8 = NamedRerouteUsage()
  n9 = ScalarParameter() {group=Gradient, parameterName=GradientStart}
  n10 = ScalarParameter() {defaultValue=1, group=Gradient, parameterName=GradientEnd}
  NamedRerouteDeclaration(None=LinearInterpolate(
    A=NamedRerouteUsage()
    B=NamedRerouteUsage()
    Alpha=MaterialFunctionCall(UV=@n8, GradientStart=@n9, GradientEnd=@n10) {materialFunction=&MF_UI_RadialGradient}.Smooth
   )) {name=BackgroundGradient, nodeColor=<rgba 0 0 0 1>}
  NamedRerouteDeclaration(None=MaterialFunctionCall!(Use Local UV, Texture Size, Scale Pivot Point) {materialFunction=/Game/Environments/FrontEnd/Blueprints/ItemPreview/VaultWorldTransitions/MaterialFunctions/MF_UI_VaultWorld_UV}.Scaled to Fit) {name=UV_ScaledToFit, nodeColor=<rgba 0 0 0 1>}
  NamedRerouteDeclaration(None=VectorParameter() {group=Color, parameterName=Color2, defaultValue=<rgba 0.1779 0.4793 0.7682 1>}.RGB) {name=BackgroundGradient_Color2, nodeColor=<rgba 0 0 0 1>}
  NamedRerouteDeclaration(None=VectorParameter() {group=Color, parameterName=Color4, defaultValue=<rgba 0.0052 0.008 0.0144 1>}.RGB) {name=Overlay_Color, nodeColor=<rgba 0 0 0 1>}
  n11 = Add(
   A=NamedRerouteUsage()
   B=AppendVector(A=ScalarParameter() {group=Gradient, parameterName=GradientXPan}, B=ScalarParameter() {group=Gradient, parameterName=GradientYPan})
  )
  n12 = ScalarParameter() {defaultValue=0.1, group=Gradient, parameterName=GradientRotation, sliderMax=0.5}
  n13 = AppendVector(
   A=ScalarParameter() {defaultValue=1.4, group=Gradient, parameterName=GradientXScale}
   B=ScalarParameter() {defaultValue=1, group=Gradient, parameterName=GradientYScale}
  )
  n14 = ScalarParameter() {defaultValue=2, group=Gradient, parameterName=GradientScale}
  NamedRerouteDeclaration(None=MaterialFunctionCall(
    UVs=MaterialFunctionCall(UVs=@n11, "Rotation Angle (0-1)"=@n12) {materialFunction=&CustomRotator}.Rotated Values
    Texture Scale=Multiply(A=@n13, B=@n14)
   ) {materialFunction=&ScaleUVsByCenter}.UVs) {name=UV_FocalPoint, nodeColor=<rgba 0 0 0 1>}
  n15 = Multiply(A=Time(), B=ScalarParameter() {defaultValue=0.5, parameterName=NoiseTime}) {constB=0.5}
  NamedRerouteDeclaration(None=Min(B=MaterialFunctionCall!(UV, Seed=@n15) {materialFunction=&MF_UI_Random}.Result) {constA=0.5}) {name=Noise, nodeColor=<rgba 0 0 0 1>}
  n16 = TextureObjectParameter() {
   group=Texture
   parameterName=CMS_Texture
   samplerSource=SSM_Wrap_WorldGroupSettings
   texture=/Game/UI/Foundation/Textures/Founders/T-VBucks-Details-5000
  }
  NamedRerouteDeclaration(None=FunctionInput(Preview=@n16) {bUsePreviewValueAsDefault=true, inputName=CMSTexture, inputType=FunctionInput_Texture2D}) {name=Texture Object, nodeColor=<rgba 0 0 0 1>}
  NamedRerouteDeclaration(None=MaterialFunctionCall(UVs=NamedRerouteUsage(), GradientStart=Constant(), GradientEnd=Constant() {r=0.5}, RotationAngle=Constant() {r=0.25}) {materialFunction=&MF_UI_LinearGradient}.Result) {name=Overlay, nodeColor=<rgba 0 0 0 1>}
  NamedRerouteDeclaration(None=Rotator(Coordinate=NamedRerouteUsage(), Time=ScalarParameter() {defaultValue=0.08, group=Global, parameterName=Tile_Angle}) {speed=6.2832}) {name=Screen_Angle, nodeColor=<rgba 0 0 0 1>}
  n17 = NamedRerouteUsage()
  n18 = AppendVector(
   A=ScalarParameter() {defaultValue=-0.3, group=Texture, parameterName=Texture_XPan}
   B=ScalarParameter() {group=Texture, parameterName=Texture_YPan}
  )
  n19 = Constant() {r=0.5}
  n20 = Subtract(
   A=MaterialFunctionCall(UVs=Add(A=@n17, B=@n18), Texture Scale=ScalarParameter() {defaultValue=1, group=Texture, parameterName=Texture_Scale}) {materialFunction=&ScaleUVsByCenter}.UVs
   B=@n19
  )
  n21 = ScalarParameter() {
   defaultValue=1
   desc="Widget aspect ratio (W/H). Corrects CMS texture aspect to prevent squishing. 1.0 = no correction."
   group=Texture
   parameterName=TargetAspectRatio
  }
  n22 = TextureProperty(Tex=@n16)
  n23 = ComponentMask(None=@n22) {r=true}
  n24 = ComponentMask(None=@n22) {g=true}
  n25 = AppendVector(A=Divide(A=@n21, B=Divide(A=@n23, B=@n24)), B=Constant() {r=1})
  n26 = TextureSample(UVs=Add(A=Multiply(A=@n20, B=@n25), B=@n19), Tex=NamedRerouteUsage()) {samplerSource=SSM_Clamp_WorldGroupSettings, texture=&T-VBucks-Details-5000}
  NamedRerouteDeclaration(None=Multiply(A=@n26.A, B=ScalarParameter() {defaultValue=1, group=Texture, parameterName=ShowTexture})) {name=Texture_Alpha, nodeColor=<rgba 0 0 0 1>}
  NamedRerouteDeclaration(None=@n26.RGB) {name=Texture_Colors, nodeColor=<rgba 0 0 0 1>}
  NamedRerouteDeclaration(None=VectorParameter() {group=Color, parameterName=Color1, defaultValue=<rgba 0.0194 0.0452 0.0685 1>}.RGB) {name=BackgroundGradient_Color1, nodeColor=<rgba 0 0 0 1>}
  TextureSampleParameter2D() {group=Texture, parameterName=CMS_Texture_O, texture=&T-VBucks-Details-5000}
 ]}
