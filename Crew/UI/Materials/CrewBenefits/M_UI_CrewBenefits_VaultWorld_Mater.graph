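# Masked, lit master material for the Crew benefits vault world, the backdrop used in the front-end item preview.
# A Crew UI gradient function supplies the base color through its Lerp output.
# That color passes through the vault world post-process calibration function, and the calibrated result feeds the shared vault world transition function.
# All logic lives in these three chained material functions.
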Material /Crew/UI/Materials/CrewBenefits/M_UI_CrewBenefits_VaultWorld_Mater
graph {
 properties {bAutomaticallySetUsageInEditor=false, blendMode=BLEND_Masked, usedShadingModels=Default Lit}
 nodes [
  n1 = MaterialFunctionCall!(Vault World Color=MaterialFunctionCall!(CMSTexture) {materialFunction=&/MF_UI_CrewGradient}.Lerp) {materialFunction=/Game/Environments/FrontEnd/Blueprints/ItemPreview/VaultWorldTransitions/MaterialFunctions/MF_UI_VaultWorld_PostProcessCalibration}
  MaterialFunctionCall!(Vault World Color=@n1.Calibrated Color) {materialFunction=/Game/Environments/FrontEnd/Blueprints/ItemPreview/VaultWorldTransitions/MaterialFunctions/MF_UI_VaultWorld_Transition}
 ]
}
